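# Sound cue for the ambient energy loop at the petrified Zero Point POI in Asteria.
# A looping night bed and a low-energy loop with its own very-close attenuation play continuously under randomly chosen whale-groan one-shots, which repeat after random delays of a few seconds.
# A quality switch uses the full three-layer mix for its first two slots and drops the groans for the last two, leaving only the loops.
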SoundCue /Game/Sounds/Asteria/POI/Zeropoint_Petrified/Spots/Spot_Zeropoint_Petrified_Occluded_Energy_Loop_Cue
root=QualityLevel_0
nodes [
 {
  id=Attenuation_0
  children [Looping_0]
  properties {attenuationSettings=&/ZP_Close_Pertrified_Att}
 }
 {
  id=Attenuation_1
  children [WavePlayer_1]
  properties {attenuationSettings=&ZP_Close_Pertrified_Att}
 }
 {
  id=Attenuation_2
  children [WavePlayer_0]
  properties {attenuationSettings=&/ZP_Close_Pertrified_Very_Close_Att}
 }
 {id=Delay_0, children [Random_1], properties {delayMax=8, delayMin=3}}
 {id=Looping_0, children [Delay_0]}
 {id=Mixer_1, children [Attenuation_1, Attenuation_0, Attenuation_2], properties {inputVolume [1.2,0.8,2]}}
 {id=Mixer_2, children [Attenuation_1, Attenuation_2], properties {inputVolume [1,1]}}
 {id=QualityLevel_0, children [Mixer_1,Mixer_1,Mixer_2,Mixer_2]}
 {
  id=Random_1
  children [
   WavePlayer_2
   WavePlayer_4
   WavePlayer_5
   WavePlayer_6
   WavePlayer_7
   WavePlayer_8
   WavePlayer_9
   WavePlayer_10
   WavePlayer_20
   WavePlayer_21
   WavePlayer_22
   WavePlayer_23
   WavePlayer_24
   WavePlayer_25
   WavePlayer_26
  ]
  properties {weights [1,1,1,1,1,1,1,1,1,1,1,1,1,1,1]}
 }
 {
  id=WavePlayer_0
  properties {bLooping=true, soundWaveAssetPtr=&/Spot_Zeropoint_Petrified_Low_Energy_Very_Close_Loop}
 }
 {
  id=WavePlayer_1
  properties {bLooping=true, soundWaveAssetPtr=&/Spot_Zeropoint_Petrified_Night_Loop}
 }
 {
  id=WavePlayer_10
  properties {soundWaveAssetPtr=&/Spot_Zeropoint_Petrified_Prehistroric_Whale_Groan_08}
 }
 {
  id=WavePlayer_2
  properties {soundWaveAssetPtr=&/Spot_Zeropoint_Petrified_Prehistroric_Whale_Groan_01}
 }
 {
  id=WavePlayer_20
  properties {soundWaveAssetPtr=&/Spot_Zeropoint_Petrified_Prehistroric_Whale_Groan_09}
 }
 {
  id=WavePlayer_21
  properties {soundWaveAssetPtr=&/Spot_Zeropoint_Petrified_Prehistroric_Whale_Groan_10}
 }
 {
  id=WavePlayer_22
  properties {soundWaveAssetPtr=&/Spot_Zeropoint_Petrified_Prehistroric_Whale_Groan_11}
 }
 {
  id=WavePlayer_23
  properties {soundWaveAssetPtr=&/Spot_Zeropoint_Petrified_Prehistroric_Whale_Groan_12}
 }
 {
  id=WavePlayer_24
  properties {soundWaveAssetPtr=&/Spot_Zeropoint_Petrified_Prehistroric_Whale_Groan_13}
 }
 {
  id=WavePlayer_25
  properties {soundWaveAssetPtr=&/Spot_Zeropoint_Petrified_Prehistroric_Whale_Groan_14}
 }
 {
  id=WavePlayer_26
  properties {soundWaveAssetPtr=&/Spot_Zeropoint_Petrified_Prehistroric_Whale_Groan_15}
 }
 {
  id=WavePlayer_4
  properties {soundWaveAssetPtr=&/Spot_Zeropoint_Petrified_Prehistroric_Whale_Groan_02}
 }
 {
  id=WavePlayer_5
  properties {soundWaveAssetPtr=&/Spot_Zeropoint_Petrified_Prehistroric_Whale_Groan_03}
 }
 {
  id=WavePlayer_6
  properties {soundWaveAssetPtr=&/Spot_Zeropoint_Petrified_Prehistroric_Whale_Groan_04}
 }
 {
  id=WavePlayer_7
  properties {soundWaveAssetPtr=&/Spot_Zeropoint_Petrified_Prehistroric_Whale_Groan_05}
 }
 {
  id=WavePlayer_8
  properties {soundWaveAssetPtr=&/Spot_Zeropoint_Petrified_Prehistroric_Whale_Groan_06}
 }
 {
  id=WavePlayer_9
  properties {soundWaveAssetPtr=&/Spot_Zeropoint_Petrified_Prehistroric_Whale_Groan_07}
 }
]
properties {
 attenuationSettings=&ZP_Close_Pertrified_Att
 maxDistance=4000
 soundClassObject=&Fort_Gadgets
 volumeMultiplier=0.8
 attenuationOverrides {
  bAttenuateWithLPF=true
  bEnableFocusInterpolation=true
  bEnableListenerFocus=true
  distanceAlgorithm=Custom
  falloffDistance=5500
  focusAttackInterpSpeed=0.5
  focusAzimuth=15
  focusDistanceScale=0.5
  focusReleaseInterpSpeed=0.25
  lPFFrequencyAtMax=500
  lPFRadiusMax=2000
  lPFRadiusMin=1000
  nonFocusAzimuth=90
  nonFocusVolumeAttenuation=0.1
  stereoSpread=2048
  attenuationShapeExtents {x=1024}
  customAttenuationCurve {editorCurveData {keys [<at 0 1>,<at 0.3 1>,<at 0.4 0.5>,<at 1 0>]}}
 }
}
